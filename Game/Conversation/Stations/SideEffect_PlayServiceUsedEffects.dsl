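; Conversation side effect that plays a service station prop's "service used" effects on the client.
; It gets the NPC participant actor from the conversation context and, if valid, casts it to BP_StationProp_Parent.
; It then calls PlayServiceUsedEffects on that prop, passing the participant's player controller.
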(fn ClientCauseSideEffect
  (params (Context :type Conversation%20Context%20Structure%20%28by%20ref%29))
  (bind _returnvalue (Class|FortConversationContextLibrary|GetParticipantActorNPC Context))
  (Utilities|IsValid _returnvalue
    (:Is%20Valid
      (bind _asbp_station_prop_parent (Utilities|Casting|CastToBP_StationProp_Parent _returnvalue))
      (bind _returnvalue_2 (Class|FortConversationContextLibrary|GetParticipantPlayerController (Variables|GetContext)))
      (Class|BPStationPropParent|PlayServiceUsedEffects _asbp_station_prop_parent _returnvalue_2))
    (:Is%20Not%20Valid
      _))
  (return))
